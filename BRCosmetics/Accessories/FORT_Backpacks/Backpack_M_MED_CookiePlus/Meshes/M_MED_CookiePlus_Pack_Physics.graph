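# Physics asset for the CookiePlus backpack cosmetic, giving capsule bodies to its dangling arms, hips and feet for secondary motion.
# The body master bone rotates freely on the pack bone and anchors two three-segment arm chains and the hips, which carry both feet.
# Angular limits narrow along each arm, hold the hips nearly rigid and let the feet swing wider, with slightly different tuning per foot.
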PhysicsAsset /BRCosmetics/Accessories/FORT_Backpacks/Backpack_M_MED_CookiePlus/Meshes/M_MED_CookiePlus_Pack_Physics
bodies {
 dyn_arm_l_01 [{shapeType=Capsule, length=5, radius=4, shapeName=dyn_arm_l_01_capsule, center=<1 0 0>, rotation=<rot 90 0 0>}]
 dyn_arm_l_02 [{shapeType=Capsule, length=5, radius=4, shapeName=dyn_arm_l_02_capsule, center=<1 0 0>, rotation=<rot 90 0 0>}]
 dyn_arm_l_03 [{shapeType=Capsule, length=5, radius=4, shapeName=dyn_arm_l_03_capsule, center=<1 0 0>, rotation=<rot 90 0 0>}]
 dyn_arm_r_01 [{shapeType=Capsule, length=5, radius=4, shapeName=dyn_arm_r_01_capsule, center=<1 0 0>, rotation=<rot 90 0 0>}]
 dyn_arm_r_02 [{shapeType=Capsule, length=5, radius=4, shapeName=dyn_arm_r_02_capsule, center=<1 0 0>, rotation=<rot 90 0 0>}]
 dyn_arm_r_03 [{shapeType=Capsule, length=5, radius=4, shapeName=dyn_arm_r_03_capsule, center=<1 0 0>, rotation=<rot 90 0 0>}]
 dyn_body_MSTR [{shapeType=Capsule, length=0.505, radius=0.505, shapeName=dyn_body_MSTR_capsule, rotation=<rot 0 0 90>}]
 dyn_foot_l [{shapeType=Capsule, length=3.5, radius=5.5, shapeName=dyn_foot_l_capsule, center=<1 0 0>, rotation=<rot 90 0 0>}]
 dyn_foot_r [{shapeType=Capsule, length=4.5, radius=4.5, shapeName=dyn_foot_r_capsule, center=<1 0 0>, rotation=<rot 90 0 0>}]
 dyn_hips [{shapeType=Capsule, length=5, radius=4, shapeName=dyn_hips_capsule, center=<1 0 0>, rotation=<rot 90 0 0>}]
 pack [{shapeType=Capsule, length=0.505, radius=0.505, shapeName=pack_capsule, rotation=<rot 0 0 90>}]
}
constraints [
 {
  bone1Name=dyn_body_MSTR
  bone2Name=pack
  swing1LimitDegrees=45
  swing1Motion=Free
  swing2LimitDegrees=45
  swing2Motion=Free
  twistLimitDegrees=45
  twistMotion=Free
 }
 {bone1Name=dyn_arm_l_01, bone2Name=dyn_body_MSTR, swing1LimitDegrees=18, swing2LimitDegrees=18, twistLimitDegrees=8}
 {bone1Name=dyn_arm_l_02, bone2Name=dyn_arm_l_01, swing1LimitDegrees=15, swing2LimitDegrees=9, twistLimitDegrees=6}
 {bone1Name=dyn_arm_l_03, bone2Name=dyn_arm_l_02, swing1LimitDegrees=15, swing2LimitDegrees=4, twistLimitDegrees=6}
 {bone1Name=dyn_arm_r_01, bone2Name=dyn_body_MSTR, swing1LimitDegrees=18, swing2LimitDegrees=18, twistLimitDegrees=8}
 {bone1Name=dyn_arm_r_02, bone2Name=dyn_arm_r_01, swing1LimitDegrees=15, swing2LimitDegrees=9, twistLimitDegrees=6}
 {bone1Name=dyn_arm_r_03, bone2Name=dyn_arm_r_02, swing1LimitDegrees=15, swing2LimitDegrees=4, twistLimitDegrees=6}
 {bone1Name=dyn_hips, bone2Name=dyn_body_MSTR, swing1LimitDegrees=3, swing2LimitDegrees=1, twistLimitDegrees=1}
 {bone1Name=dyn_foot_l, bone2Name=dyn_hips, swing1LimitDegrees=24, swing2LimitDegrees=16, twistLimitDegrees=2}
 {bone1Name=dyn_foot_r, bone2Name=dyn_hips, swing1LimitDegrees=32, swing2LimitDegrees=12, twistLimitDegrees=2}
]
